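# Unlit translucent two-sided Niagara sprite material for the DonkeyCrib emote's screen effect.
# Emissive sums a background color, screen texture, center-weighted background noise and masked edge smoke, then multiplies by particle color into MF_FX_EmissiveAdjustments.
# Each scrolling texture's UVAndSpeed parameter uses only R for tiling and B for speed, both axes alike.
# Particle alpha times Global_Opacity feeds a DepthFade, independent of the textures.
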Material /BRCosmetics/Animation/Game/MainPlayer/Emotes/DonkeyCrib/VFX/Material/M_DonkeyCrib_Screen
graph {
 properties {
  bAutomaticallySetUsageInEditor=false
  bUsedWithNiagaraSprites=true
  blendMode=BLEND_Translucent
  shadingModel=MSM_Unlit
  twoSided=true
  usedShadingModels=Unlit
 }
 nodes [
  n1 = ParticleColor()
  DepthFade(
   Opacity=Multiply(A=@n1.A, B=ScalarParameter() {defaultValue=1, group=Opacity, parameterName=Global_Opacity})
   FadeDistance=ScalarParameter() {group=Opacity, parameterName=DepthFade}
  )
  n2 = VectorParameter() {group=BackGround, parameterName=Emissive_Color_BackGround, defaultValue=<rgba 1 1 1 1>}
  n3 = ScalarParameter() {defaultValue=1, group=BackGround, parameterName=Emissive_Opacity_Bakcground}
  n4 = Add(
   A=TextureCoordinate()
   B=AppendVector(A=ScalarParameter() {group=Main, parameterName=Main_Texture_U_Offset}, B=ScalarParameter() {group=Main, parameterName=Main_Texture_V_Offset})
  )
  n5 = VectorParameter() {group=Main, parameterName=Main_Texture_UVAndSpeed, defaultValue=<rgba 1 1 0 0>}
  n6 = AppendVector(A=@n5.R, B=@n5.R)
  n7 = AppendVector(A=@n5.B, B=@n5.B)
  n8 = Time()
  n9 = TextureSampleParameter2D(UVs=Add(A=Multiply(A=@n4, B=@n6), B=Multiply(A=@n7, B=@n8))) {
   group=Main
   parameterName=Main_Texture
   texture=/BRCosmetics/Animation/Game/MainPlayer/Emotes/DonkeyCrib/VFX/Texture/T_DonkeyCrib_Screen_05_04
  }
  n10 = Multiply(A=@n9.RGB, B=VectorParameter() {group=Main, parameterName=Main_Texture_Color, defaultValue=<rgba 1 1 1 1>}.RGB)
  n11 = ScalarParameter() {defaultValue=1, group=Main, parameterName=Main_Texture_Opacity}
  n12 = Add(
   A=TextureCoordinate()
   B=AppendVector(
    A=ScalarParameter() {group=BackGroundNoise, parameterName=BackGroundNoise_Texture_U_Offset}
    B=ScalarParameter() {group=BackGroundNoise, parameterName=BackGroundNoise_Texture_V_Offset}
   )
  )
  n13 = VectorParameter() {group=BackGroundNoise, parameterName=BackGroundNoise_Texture_UVAndSpeed, defaultValue=<rgba 1 1 0 0>}
  n14 = AppendVector(A=@n13.R, B=@n13.R)
  n15 = AppendVector(A=@n13.B, B=@n13.B)
  n16 = Time()
  n17 = TextureSampleParameter2D(UVs=Add(A=Multiply(A=@n12, B=@n14), B=Multiply(A=@n15, B=@n16))) {group=BackGroundNoise, parameterName=BackGroundNoise_Texture, texture=&AtmosphericCloudNoise01}
  n18 = Multiply(A=@n17.RGB, B=VectorParameter() {group=BackGroundNoise, parameterName=BackGroundNoise_Color, defaultValue=<rgba 1 1 1 1>}.RGB)
  n19 = ComponentMask(None=TextureCoordinate()) {g=true, r=true}
  n20 = MaterialFunctionCall(Float2=Multiply(A=OneMinus(None=@n19), B=@n19)) {materialFunction=&BreakOutFloat2Components}
  n21 = Multiply(
   A=Power(Base=@n20.R, Exp=ScalarParameter() {defaultValue=1, group=BackGroundNoise, parameterName=BackGroundNoise_Opacity_R_Power})
   B=Power(Base=@n20.G, Exp=ScalarParameter() {defaultValue=1, group=BackGroundNoise, parameterName=BackGroundNoise_Opacity_G_Power})
  )
  n22 = Add(
   A=Add(A=Multiply(A=@n2.RGB, B=@n3) {constB=0}, B=Multiply(A=@n10, B=@n11) {constB=0})
   B=Multiply(A=Multiply(A=@n18, B=@n21), B=ScalarParameter() {defaultValue=1, group=BackGroundNoise, parameterName=BackGroundNoise_Opacity}) {constB=0}
  )
  n23 = Add(
   A=TextureCoordinate()
   B=AppendVector(
    A=ScalarParameter() {group=EdgeNoise, parameterName=EdgeNoise_Texture_U_Offset}
    B=ScalarParameter() {group=EdgeNoise, parameterName=EdgeNoise_Texture_V_Offset}
   )
  )
  n24 = VectorParameter() {group=EdgeNoise, parameterName=EdgeNoise_Texture_UVAndSpeed, defaultValue=<rgba 1 1 0 0>}
  n25 = AppendVector(A=@n24.R, B=@n24.R)
  n26 = AppendVector(A=@n24.B, B=@n24.B)
  n27 = Time()
  n28 = TextureSampleParameter2D(UVs=Add(A=Multiply(A=@n23, B=@n25), B=Multiply(A=@n26, B=@n27))) {group=EdgeNoise, parameterName=EdgeNoise_Texture, texture=/Game/Global/Textures/Default/Tileable_Patterns/T_Thick_Smoke_05}
  n29 = VectorParameter() {group=EdgeNoise, parameterName=EdgeNoise_Texture_Color, defaultValue=<rgba 1 1 1 1>}
  n30 = TextureSampleParameter2D() {
   group=EdgeNoise
   parameterName=EdgeNoise_Texture_Mask
   texture=/BRCosmetics/Animation/Game/MainPlayer/Emotes/DonkeyCrib/VFX/Texture/T_DonkeyCrib_Mask
  }
  n31 = Multiply(
   A=Multiply(A=Multiply(A=@n28.RGB, B=@n29.RGB), B=@n30.RGB)
   B=ScalarParameter() {defaultValue=1, group=EdgeNoise, parameterName=EdgeNoise_Opacity}
  ) {constB=0}
  MaterialFunctionCall(Emissive=Multiply(A=@n1.RGB, B=Add(A=@n22, B=@n31))) {materialFunction=&MF_FX_EmissiveAdjustments}
 ]
}
