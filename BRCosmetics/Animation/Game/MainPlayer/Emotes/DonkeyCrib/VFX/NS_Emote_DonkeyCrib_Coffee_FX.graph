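# Niagara system for the coffee effect of the DonkeyCrib player emote, named for a coffee splash with droplets.
# It contains two emitters, Splash and Droplet, with no renderers listed for either.
# SystemState is the only module recorded, and the system exposes no user variables.
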NiagaraSystem /BRCosmetics/Animation/Game/MainPlayer/Emotes/DonkeyCrib/VFX/NS_Emote_DonkeyCrib_Coffee_FX
dependencies {usedDataInterfaces [], usedDynamicInputs [], usedModules [&SystemState]}
emitters {Droplet {rendererClasses [], renderers []}, Splash {rendererClasses [], renderers []}}
summary {
 systemName=NS_Emote_DonkeyCrib_Coffee_FX
 emitters [{emitterName=Splash, rendererClasses []}, {emitterName=Droplet, rendererClasses []}]
 userVariables []
}
