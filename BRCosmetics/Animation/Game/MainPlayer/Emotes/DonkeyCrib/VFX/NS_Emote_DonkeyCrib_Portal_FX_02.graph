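# Niagara system for the second portal effect of the DonkeyCrib main player emote.
# It has Cylinder and Flare emitters, plus a disabled Glow emitter, under a standard SystemState module.
# User float parameters Opacity and Size are exposed, presumably to tune the portal's fade and scale.
# No renderers, data interfaces or dynamic inputs are listed.
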NiagaraSystem /BRCosmetics/Animation/Game/MainPlayer/Emotes/DonkeyCrib/VFX/NS_Emote_DonkeyCrib_Portal_FX_02
dependencies {usedDataInterfaces [], usedDynamicInputs [], usedModules [&SystemState]}
emitters {
 Cylinder {rendererClasses [], renderers []}
 Flare {rendererClasses [], renderers []}
 Glow {bEnabled=false, rendererClasses [], renderers []}
}
summary {
 systemName=NS_Emote_DonkeyCrib_Portal_FX_02
 emitters [{bEnabled=false, emitterName=Glow, rendererClasses []}, {emitterName=Flare, rendererClasses []}, {emitterName=Cylinder, rendererClasses []}]
 userVariables [
  User.Opacity {type=NiagaraFloat, defaultValue=1, description=""}
  User.Size {type=NiagaraFloat, defaultValue=5, description=""}
 ]
}
