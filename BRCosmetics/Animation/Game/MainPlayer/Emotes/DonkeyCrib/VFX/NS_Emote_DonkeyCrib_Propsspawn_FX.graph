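# Niagara system for the prop-spawn effect of the DonkeyCrib main-player emote.
# It layers sparkle, airglow and smoke emitters, whose LWE suffix suggests lightweight emitters, plus a second airglow emitter that is disabled.
# The system uses SystemState, a spawn-time variable assignment and the FNTimeOfDayColor dynamic input for time-of-day color.
# It exposes no user parameters.
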NiagaraSystem /BRCosmetics/Animation/Game/MainPlayer/Emotes/DonkeyCrib/VFX/NS_Emote_DonkeyCrib_Propsspawn_FX
dependencies {
 usedDataInterfaces []
 usedDynamicInputs [&FNTimeOfDayColor]
 usedModules [
  &SystemState
  &self:SystemSpawnScript.SystemScriptSource.SystemScriptGraph.NiagaraNodeAssignment_0.SetVariables_#1
 ]
}
emitters {
 AirglowLWE {rendererClasses [], renderers []}
 AirglowLWE001 {bEnabled=false, rendererClasses [], renderers []}
 SmokeLWE {rendererClasses [], renderers []}
 SparklesLWE {rendererClasses [], renderers []}
}
summary {
 systemName=NS_Emote_DonkeyCrib_Propsspawn_FX
 emitters [
  {emitterName=SparklesLWE, rendererClasses []}
  {emitterName=AirglowLWE, rendererClasses []}
  {emitterName=SmokeLWE, rendererClasses []}
  {bEnabled=false, emitterName=AirglowLWE001, rendererClasses []}
 ]
 userVariables []
}
