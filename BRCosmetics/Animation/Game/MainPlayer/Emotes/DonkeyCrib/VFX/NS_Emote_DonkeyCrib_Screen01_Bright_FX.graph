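# Niagara system for the DonkeyCrib emote, named as a bright effect for its first screen.
# Its only emitter, Bright, is disabled and lists no renderers, so the system draws nothing as it stands.
# SystemState is the only module in use, with no data interfaces or dynamic inputs.
# It exposes a FrontendPos vector user parameter, zero by default, with no visible consumer.
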NiagaraSystem /BRCosmetics/Animation/Game/MainPlayer/Emotes/DonkeyCrib/VFX/NS_Emote_DonkeyCrib_Screen01_Bright_FX
dependencies {usedDataInterfaces [], usedDynamicInputs [], usedModules [&SystemState]}
emitters {Bright {bEnabled=false, rendererClasses [], renderers []}}
summary {
 systemName=NS_Emote_DonkeyCrib_Screen01_Bright_FX
 emitters [{bEnabled=false, emitterName=Bright, rendererClasses []}]
 userVariables [User.FrontendPos {type=Vector3f, description="", defaultValue {struct=Vector3f, value=<0 0 0>}}]
}
