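# Niagara effect for screen 01 of the DonkeyCrib player emote: a glowing sprite display that scales in, with animated overlays.
# MainScreen001 bursts three spawn groups with visibility tags for two sprite renderers and a light renderer; Glasses adds a fading sprite.
# Screen_01_Head follows a curved path, feeding Trail_01's ribbon through particle reads; Flare has no renderers.
# User floats control element opacity, glass opacity and blue tint, while time-of-day scalars adjust some alphas.
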NiagaraSystem /BRCosmetics/Animation/Game/MainPlayer/Emotes/DonkeyCrib/VFX/NS_Emote_DonkeyCrib_Screen01_FX
dependencies {
 usedDataInterfaces [
  &NiagaraDataInterfaceCurve
  &NiagaraDataInterfaceParticleRead
  &NiagaraDataInterfaceVector2DCurve
 ]
 usedDynamicInputs []
 usedModules [&SystemState]
}
emitters {
 Flare {rendererClasses [], renderers []}
 Glasses {
  emitterUpdateScript [
   &EmitterState(Life Cycle Mode=::NewEnumerator1, Loop Duration=3, UseLoopDelay=false)
   &SpawnBurst_Instantaneous(Spawn Time=1, Use Spawn Probability=false, UseLoopCountLimit=false)
  ]
  particleSpawnScript [
   &InitializeParticle(
    Color=<rgba 1 1 1 0.5>
    Lifetime=2
    Position Offset=<0 -1 -6.5>
    Sprite Size=<47 17>
    Sprite Size Mode=::NewEnumerator4
    UsePositionOffset=true
    Write Material Random=true
   )
   &SpriteFacingAndAlignment(Sprite Alignment=<0 0 1>, Sprite Facing=<-1 0 0>, WriteAlignment=true, WriteFacing=true)
  ]
  particleUpdateScript [
   &ParticleState()
   &ScaleColor(
    Scale Alpha=&FloatFromCurve(FloatCurve=DataInterface() {propertyValues {Curve {keys [
         {interpMode=RCIM_Cubic, tangentMode=RCTM_Break, time=0, value=0}
         {interpMode=RCIM_Cubic, tangentMode=RCTM_Break, time=0.5, value=1}
         {interpMode=RCIM_Cubic, tangentMode=RCTM_Break, time=1, value=0}
        ]}}})
    ScaleA=true
    ScaleRGB=true
   )
   &ScaleColor(Scale Alpha=$User.Screen01_Glass_Opacity, ScaleA=true, ScaleRGB=true) {module=ScaleColor001}
  ]
  rendererClasses [&NiagaraSpriteRendererProperties]
  renderers [{rendererClass=&NiagaraSpriteRendererProperties}]
 }
 MainScreen001 {
  emitterUpdateScript [
   &EmitterState(Life Cycle Mode=::NewEnumerator1, Loop Behavior=::NewEnumerator1, Loop Duration Mode=::NewEnumerator1, UseLoopDelay=false)
   &SpawnBurst_Instantaneous(Use Spawn Probability=false, UseLoopCountLimit=false)
   &SpawnBurst_Instantaneous(Spawn Group=1, Use Spawn Probability=false, UseLoopCountLimit=false) {module=SpawnBurst_Instantaneous001}
   &SpawnBurst_Instantaneous(Spawn Group=2, Use Spawn Probability=false, UseLoopCountLimit=false) {module=SpawnBurst_Instantaneous002}
  ]
  particleSpawnScript [
   &InitializeParticle(
    Color=&MaskLinearColorBySpawnGroup(
     Masked Linear Color=<rgba 1 1 1 1>
     Passthrough Linear Color=&MaskLinearColorBySpawnGroup(Masked Linear Color=<rgba 0.197 0.631 1 1>, Passthrough Linear Color=<rgba 0.696 3.242 6 1>, Spawn Group Mask=1)
    )
    Lifetime=5
    Position=&MaskVectorBySpawnGroup(
     Masked Vector=<0 0 0>
     Passthrough Vector=&MaskVectorBySpawnGroup(Masked Vector=<0 -15 3>, Passthrough Vector=<0 0 0>, Spawn Group Mask=1)
    )
    Position Mode=::NewEnumerator1
    Sprite Size=&MaskVector2dBySpawnGroup(Masked Vector2D=<60 30>, Passthrough Vector2D=<7 24>, Spawn Group Mask=0)
    Sprite Size Mode=::NewEnumerator4
    UsePositionOffset=false
    Write Material Random=true
   )
   /SetVariables(Particles.VisibilityTag=&SpawnGroup()) {isSetParameters=true}
   &SpriteFacingAndAlignment(
    Sprite Alignment=&MaskVectorBySpawnGroup(Masked Vector=<0 0 1>, Passthrough Vector=<0 1 0>)
    Sprite Facing=<-1 0 0>
    WriteAlignment=true
    WriteFacing=true
   )
  ]
  particleUpdateScript [
   &ParticleState(Kill Particles When Lifetime Has Elapsed=false)
   &ScaleSpriteSize(
    Non-Uniform Curve Sprite Scale=DataInterface() {propertyValues {
      XCurve {keys [
        {interpMode=RCIM_Cubic, tangentMode=RCTM_Break, time=0, value=0}
        {arriveTangent=2.8331, interpMode=RCIM_Cubic, leaveTangent=25.7439, tangentMode=RCTM_Break, time=0.1498, value=0.1158}
        {arriveTangent=-0.0431, interpMode=RCIM_Cubic, tangentMode=RCTM_Break, time=0.1787, value=0.9863}
       ]}
      YCurve {keys [
        <at 0 0 RCIM_Cubic>
        {interpMode=RCIM_Cubic, leaveTangent=15.5999, tangentMode=RCTM_Break, time=0.0231, value=0.1813}
        {
         arriveTangentWeight=0.3394
         interpMode=RCIM_Cubic
         leaveTangent=31.0823
         leaveTangentWeight=0.078
         tangentMode=RCTM_Break
         time=0.1101
         value=0.8778
        }
        {arriveTangent=0.1208, interpMode=RCIM_Cubic, leaveTangent=-13.8768, tangentMode=RCTM_Break, time=0.1248, value=1.2343}
        {arriveTangent=-2.2265, interpMode=RCIM_Cubic, leaveTangent=-2.2265, tangentMode=RCTM_Break, time=0.1475, value=1}
        {interpMode=RCIM_Cubic, leaveTangent=1.8745, tangentMode=RCTM_Break, time=0.23, value=1}
       ]}
     }}
    Scale Sprite Size Mode=::NewEnumerator3
   ) {module=ScaleSpriteSize001}
   &ScaleColor(
    Scale Alpha=&MaskFloatBySpawnGroup(
     Masked Float=&FloatFromCurve(
      FloatCurve=DataInterface() {propertyValues {Curve {keys [{interpMode=RCIM_Cubic, tangentMode=RCTM_Break, time=0, value=0}, {interpMode=RCIM_Cubic, tangentMode=RCTM_Break, time=1, value=1}]}}}
      CurveIndex=$Engine.System.Age
      Scale Curve=&Multiply_Float(
       A=0.8
       B=&FNTimeOfDayScalar(Night Scale=0.8, Morning Scale=0.8, Day Scale=1.3, Evening Scale=0.8)
      )
     )
     Passthrough Float=&FloatFromCurve(
      FloatCurve=DataInterface() {propertyValues {Curve {keys [{interpMode=RCIM_Cubic, tangentMode=RCTM_Break, time=1, value=0}, {interpMode=RCIM_Cubic, tangentMode=RCTM_Break, time=2, value=1}]}}}
      CurveIndex=$Engine.System.Age
      Scale Curve=0.8
     )
    )
    ScaleA=true
    ScaleRGB=true
   )
   &ScaleColor(
    Scale Alpha=&MaskFloatBySpawnGroup(
     Masked Float=$User.Screen01_Element_Opacity
     Passthrough Float=&MaskFloatBySpawnGroup(Masked Float=$User.Screen01_Glass_Opacity, Passthrough Float=1, Spawn Group Mask=1)
    )
    Scale RGB=&MaskVectorBySpawnGroup(
     Masked Vector=<1 1 1>
     Passthrough Vector=&MaskVectorBySpawnGroup(
      Masked Vector=<1 1 1>
      Passthrough Vector=&MakeVector(X=1, Y=1, Z=$User.Screen01_Light_Blue)
      Spawn Group Mask=1
     )
    )
    ScaleA=true
    ScaleRGB=true
   ) {module=ScaleColor001}
  ]
  rendererClasses [&NiagaraSpriteRendererProperties, &NiagaraLightRendererProperties]
  renderers [
   {rendererClass=&NiagaraSpriteRendererProperties}
   {rendererClass=&NiagaraSpriteRendererProperties, rendererIndex=1}
   {rendererClass=&NiagaraLightRendererProperties, rendererIndex=2}
  ]
 }
 Screen_01_Head {
  emitterUpdateScript [
   &EmitterState(Life Cycle Mode=::NewEnumerator1, Loop Duration=6, UseLoopDelay=false)
   &SpawnBurst_Instantaneous(Spawn Time=1, Use Spawn Probability=false, UseLoopCountLimit=false)
  ]
  particleSpawnScript [&InitializeParticle(
    Color=<rgba 0.0791 0.5089 1 1>
    Lifetime=3
    Position Mode=::NewEnumerator0
    Sprite Size Mode=::NewEnumerator3
    Uniform Sprite Size=12
    Write Material Random=true
   )]
  particleUpdateScript [
   &ParticleState()
   &SystemLocation(MaskSpawnGroup=false, Offset=<0 9 -0.5>) {module=SystemLocation002}
   &SystemLocation(
    MaskSpawnGroup=false
    Offset=&AddVectorToPosition(
     Position=$Particles.Position
     Vector=&MakeVector(
      X=0
      Y=&FloatFromCurve(FloatCurve=DataInterface() {propertyValues {Curve {keys [{interpMode=RCIM_Cubic, tangentMode=RCTM_Break, time=0, value=0}, {interpMode=RCIM_Cubic, tangentMode=RCTM_Break, time=1, value=15}]}}})
      Z=&FloatFromCurve(FloatCurve=DataInterface() {propertyValues {Curve {keys [
           {interpMode=RCIM_Cubic, tangentMode=RCTM_Break, time=0, value=0}
           {interpMode=RCIM_Cubic, tangentMode=RCTM_Break, time=0.3, value=3}
           {interpMode=RCIM_Cubic, tangentMode=RCTM_Break, time=0.5, value=2.8}
           {interpMode=RCIM_Cubic, tangentMode=RCTM_Break, time=0.65, value=3}
           {interpMode=RCIM_Cubic, tangentMode=RCTM_Break, time=0.8, value=5.5}
           {interpMode=RCIM_Cubic, tangentMode=RCTM_Break, time=1, value=6}
          ]}}})
     )
    )
   ) {module=SystemLocation003}
   &ScaleColor(
    Scale Alpha=&FloatFromCurve(
     FloatCurve=DataInterface() {propertyValues {Curve {keys [
         {interpMode=RCIM_Cubic, tangentMode=RCTM_Break, time=0, value=0}
         {interpMode=RCIM_Cubic, tangentMode=RCTM_Break, time=0.2, value=1}
         {interpMode=RCIM_Cubic, tangentMode=RCTM_Break, time=0.8, value=1}
         {interpMode=RCIM_Cubic, tangentMode=RCTM_Break, time=1, value=0}
        ]}}}
     Scale Curve=&Multiply_Float(A=0.5, B=&FNTimeOfDayScalar(Night Scale=1, Morning Scale=1, Day Scale=1.3, Evening Scale=1))
    )
    ScaleA=true
    ScaleRGB=true
   )
   &ScaleColor(Scale Alpha=$User.Screen01_Element_Opacity, ScaleA=true, ScaleRGB=true) {module=ScaleColor001}
  ]
  rendererClasses [&NiagaraSpriteRendererProperties]
  renderers [{rendererClass=&NiagaraSpriteRendererProperties}]
 }
 Trail_01 {
  emitterUpdateScript [
   &EmitterState(Life Cycle Mode=::NewEnumerator1, Loop Duration=4, UseLoopDelay=false)
   &SpawnParticlesFromOtherEmitter(
    Attribute Reader=DataInterface() {propertyValues {EmitterBinding {bindingMode=Other, emitterName=Screen_01_Head}}}
    Calculate Spawn Rate Per Particle=false
    Maximum Age Spawn Threshold=100000
    Minimum Age Spawn Threshold=0
    Spawn Rate=20
    Use Maximum Age=true
    Use Minimum Age=true
   )
  ]
  particleSpawnScript [
   &InitializeParticle(
    Color=<rgba 0.0791 0.5089 1 1>
    Lifetime=4
    Ribbon Facing Mode=::NewEnumerator1
    Ribbon Facing Vector:Vector3f=<1 0 0>
    Ribbon Width=0.3
    Ribbon Width Mode=::NewEnumerator1
    Sprite Size Mode=::NewEnumerator3
    Uniform Sprite Size=0.3
    UsePositionOffset=false
    Write Material Random=true
   )
   &SampleParticlesFromOtherEmitter(Particle ID Sampling=::NewEnumerator1)
  ]
  particleUpdateScript [
   &ParticleState()
   &SolveForcesAndVelocity(Clamp Velocity=false, Limit Acceleration=false)
   &ScaleColor(
    Scale Alpha=&FloatFromCurve(
     FloatCurve=DataInterface() {propertyValues {Curve {keys [
         {interpMode=RCIM_Cubic, tangentMode=RCTM_Break, time=0, value=0}
         {interpMode=RCIM_Cubic, tangentMode=RCTM_Break, time=0.5, value=1}
         {interpMode=RCIM_Cubic, tangentMode=RCTM_Break, time=1, value=0}
        ]}}}
     Scale Curve=&Multiply_Float(A=1, B=&FNTimeOfDayScalar(Night Scale=1, Morning Scale=1, Day Scale=1.3, Evening Scale=1))
    )
    ScaleA=true
    ScaleRGB=true
   )
   &ScaleColor(Scale Alpha=$User.Screen01_Element_Opacity, ScaleA=true, ScaleRGB=true) {module=ScaleColor001}
  ]
  rendererClasses [&NiagaraRibbonRendererProperties]
  renderers [{rendererClass=&NiagaraRibbonRendererProperties}]
 }
}
summary {
 systemName=NS_Emote_DonkeyCrib_Screen01_FX
 emitters [
  {emitterName=Screen_01_Head, rendererClasses [&NiagaraSpriteRendererProperties]}
  {emitterName=Trail_01, rendererClasses [&NiagaraRibbonRendererProperties]}
  {emitterName=Flare, rendererClasses []}
  {emitterName=MainScreen001, rendererClasses [&NiagaraSpriteRendererProperties, &NiagaraLightRendererProperties]}
  {emitterName=Glasses, rendererClasses [&NiagaraSpriteRendererProperties]}
 ]
 userVariables [
  User.Screen01_Element_Opacity {type=NiagaraFloat, defaultValue=1, description=""}
  User.Screen01_Glass_Opacity {type=NiagaraFloat, defaultValue=1, description=""}
  User.Screen01_Light_Blue {type=NiagaraFloat, defaultValue=1, description=""}
 ]
}
