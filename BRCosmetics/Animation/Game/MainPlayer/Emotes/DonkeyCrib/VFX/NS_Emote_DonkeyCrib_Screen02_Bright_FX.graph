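# Niagara system for the DonkeyCrib player emote, named as a bright effect for its second screen.
# It holds a Bright_Sphere emitter and a disabled Light emitter.
# The file lists no renderers, data interfaces or user variables, and SystemState is the only module recorded, so it does not show how the emitters produce their look.
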NiagaraSystem /BRCosmetics/Animation/Game/MainPlayer/Emotes/DonkeyCrib/VFX/NS_Emote_DonkeyCrib_Screen02_Bright_FX
dependencies {usedDataInterfaces [], usedDynamicInputs [], usedModules [&SystemState]}
emitters {Bright_Sphere {rendererClasses [], renderers []}, Light {bEnabled=false, rendererClasses [], renderers []}}
summary {
 systemName=NS_Emote_DonkeyCrib_Screen02_Bright_FX
 emitters [{emitterName=Bright_Sphere, rendererClasses []}, {bEnabled=false, emitterName=Light, rendererClasses []}]
 userVariables []
}
