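# Niagara system for the second screen effect in the DonkeyCrib player emote, showing an animated blue screen with a spaceship.
# MainScreen pops in sized sprites and a light, with time-of-day alpha and a user float scaling blue.
# Spaceship layers fading sprites by spawn group, while Head, Lead and Screen_02_Head follow sine or curve paths that ribbon trails sample via particle readers.
# Flare, text, bar and number emitters have no renderers.
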NiagaraSystem /BRCosmetics/Animation/Game/MainPlayer/Emotes/DonkeyCrib/VFX/NS_Emote_DonkeyCrib_Screen02_FX
dependencies {
 usedDataInterfaces [
  &NiagaraDataInterfaceCurve
  &NiagaraDataInterfaceVector2DCurve
  &NiagaraDataInterfaceParticleRead
 ]
 usedDynamicInputs [&FNTimeOfDayColor]
 usedModules [
  &self:SystemSpawnScript.SystemScriptSource.SystemScriptGraph.NiagaraNodeAssignment_0.SetVariables_#1
  &self:SystemSpawnScript.SystemScriptSource.SystemScriptGraph.NiagaraNodeAssignment_1.SetVariables_#2
  &SystemState
 ]
}
emitters {
 Bar {rendererClasses [], renderers []}
 Flare {rendererClasses [], renderers []}
 Head {
  emitterUpdateScript [
   &EmitterState(Life Cycle Mode=::NewEnumerator1, Loop Duration=12.5, UseLoopDelay=false)
   &SpawnBurst_Instantaneous(Use Spawn Probability=false, UseLoopCountLimit=false)
  ]
  particleSpawnScript [&InitializeParticle(
    Color Mode=::NewEnumerator0
    Lifetime=12.5
    Sprite Size Mode=::NewEnumerator3
    Uniform Sprite Size=1
    UsePositionOffset=false
    Write Material Random=true
   )]
  particleUpdateScript [
   &ParticleState()
   &SystemLocation(
    MaskSpawnGroup=false
    Offset=&MakeVector(
     X=0
     Y=&Sine(Normalized Angle=$Emitter.Age, Period=12.5, Scale=2.5, Bias=0)
     Z=&Cosine(Normalized Angle=$Emitter.Age, Period=12.5, Scale=2.5, Bias=0)
    )
   )
   &SystemLocation(MaskSpawnGroup=false, Offset=&AddVectorToPosition(Position=$Particles.Position, Vector=<0 -22.3 6.5>)) {module=SystemLocation001}
  ]
  rendererClasses [&NiagaraSpriteRendererProperties]
  renderers [{rendererClass=&NiagaraSpriteRendererProperties}]
 }
 Lead {
  emitterUpdateScript [
   &EmitterState(Life Cycle Mode=::NewEnumerator1, Loop Duration=3, UseLoopDelay=false)
   &SpawnBurst_Instantaneous(Spawn Time=1, Use Spawn Probability=false, UseLoopCountLimit=false)
  ]
  particleSpawnScript [
   &InitializeParticle(
    Color Mode=::NewEnumerator0
    Lifetime=5
    Sprite Size Mode=::NewEnumerator3
    Uniform Sprite Size=1
    UsePositionOffset=false
    Write Material Random=true
   )
   &CameraOffset()
  ]
  particleUpdateScript [
   &ParticleState()
   &SystemLocation(
    MaskSpawnGroup=false
    Offset=&MakeVector(
     X=0
     Y=&FloatFromCurve(FloatCurve=DataInterface() {propertyValues {Curve {keys [
          {interpMode=RCIM_Cubic, leaveTangent=-2.2236, tangentMode=RCTM_Break, time=0, value=2}
          {arriveTangent=-20.7702, interpMode=RCIM_Cubic, tangentMode=RCTM_Break, time=0.45, value=-5}
          {arriveTangent=-2.4831, interpMode=RCIM_Cubic, leaveTangent=0.7848, tangentMode=RCTM_Break, time=0.6, value=-5}
          {arriveTangent=42.5759, interpMode=RCIM_Cubic, tangentMode=RCTM_Break, time=1, value=6.5}
         ]}}})
     Z=&FloatFromCurve(FloatCurve=DataInterface() {propertyValues {Curve {keys [
          {arriveTangent=-6.4111, interpMode=RCIM_Cubic, leaveTangent=11.3756, tangentMode=RCTM_Break, time=0, value=0}
          {arriveTangent=-6.4111, interpMode=RCIM_Cubic, leaveTangent=-0.588, tangentMode=RCTM_Break, time=0.45, value=4}
          {arriveTangent=0.3545, interpMode=RCIM_Cubic, leaveTangent=10.0927, tangentMode=RCTM_Break, time=0.6, value=4}
          {arriveTangent=-4.5008, interpMode=RCIM_Cubic, tangentMode=RCTM_Break, time=1, value=6.5}
         ]}}})
    )
   )
  ]
  rendererClasses [&NiagaraSpriteRendererProperties]
  renderers [{rendererClass=&NiagaraSpriteRendererProperties}]
 }
 MainScreen {
  emitterUpdateScript [
   &EmitterState(Life Cycle Mode=::NewEnumerator1, Loop Behavior=::NewEnumerator1, Loop Duration=1, UseLoopDelay=false)
   &SpawnBurst_Instantaneous(Use Spawn Probability=false, UseLoopCountLimit=false)
   &SpawnBurst_Instantaneous(Spawn Group=1, Use Spawn Probability=false, UseLoopCountLimit=false) {module=SpawnBurst_Instantaneous001}
   &SpawnBurst_Instantaneous(Spawn Group=2, Use Spawn Probability=false, UseLoopCountLimit=false) {module=SpawnBurst_Instantaneous002}
  ]
  particleSpawnScript [
   &InitializeParticle(
    Color=&MaskLinearColorBySpawnGroup(
     Masked Linear Color=<rgba 1 1 1 1>
     Passthrough Linear Color=&MaskLinearColorBySpawnGroup(Masked Linear Color=<rgba 0.197 0.631 1 1>, Passthrough Linear Color=<rgba 1.161 5.402 10 1>, Spawn Group Mask=1)
    )
    Lifetime=5
    Position=&MaskVectorBySpawnGroup(Masked Vector=<0 0 0>, Passthrough Vector=<0 -17.3 -10>)
    Position Mode=::NewEnumerator1
    Sprite Size=&MaskVector2dBySpawnGroup(Masked Vector2D=<60 30>, Passthrough Vector2D=<2.4 20>, Spawn Group Mask=0)
    Sprite Size Mode=::NewEnumerator4
    UsePositionOffset=false
    Write Material Random=true
   )
   /SetVariables(Particles.VisibilityTag=&SpawnGroup()) {isSetParameters=true}
   &SpriteFacingAndAlignment(
    Sprite Alignment=&MaskVectorBySpawnGroup(Masked Vector=<0 0 1>, Passthrough Vector=<0 1 0>)
    Sprite Facing=&MaskVectorBySpawnGroup(Masked Vector=<-1 0 0>, Passthrough Vector=<1 0 0>)
    WriteAlignment=true
    WriteFacing=true
   )
  ]
  particleUpdateScript [
   &ParticleState(Kill Particles When Lifetime Has Elapsed=false)
   &ScaleSpriteSize(
    Non-Uniform Curve Sprite Scale=DataInterface() {propertyValues {
      XCurve {keys [
        {interpMode=RCIM_Cubic, tangentMode=RCTM_Break, time=0, value=0}
        {arriveTangent=2.8331, interpMode=RCIM_Cubic, leaveTangent=25.7439, tangentMode=RCTM_Break, time=0.1498, value=0.1158}
        {arriveTangent=-0.0431, interpMode=RCIM_Cubic, tangentMode=RCTM_Break, time=0.1787, value=0.9863}
       ]}
      YCurve {keys [
        <at 0 0 RCIM_Cubic>
        {interpMode=RCIM_Cubic, leaveTangent=15.5999, tangentMode=RCTM_Break, time=0.0231, value=0.1813}
        {
         arriveTangentWeight=0.3394
         interpMode=RCIM_Cubic
         leaveTangent=31.0823
         leaveTangentWeight=0.078
         tangentMode=RCTM_Break
         time=0.1101
         value=0.8778
        }
        {arriveTangent=0.1208, interpMode=RCIM_Cubic, leaveTangent=-13.8768, tangentMode=RCTM_Break, time=0.1248, value=1.2343}
        {arriveTangent=-2.2265, interpMode=RCIM_Cubic, leaveTangent=-2.2265, tangentMode=RCTM_Break, time=0.1475, value=1}
        {interpMode=RCIM_Cubic, leaveTangent=1.8745, tangentMode=RCTM_Break, time=0.23, value=1}
       ]}
     }}
    Scale Sprite Size Mode=::NewEnumerator3
   ) {module=ScaleSpriteSize001}
   &ScaleColor(
    Scale Alpha=&MaskFloat(
     Mask Source=0
     Mask Parameter=?
     Mask=0
     Masked Float=&FloatFromCurve(
      FloatCurve=DataInterface() {propertyValues {Curve {keys [{interpMode=RCIM_Cubic, tangentMode=RCTM_Break, time=0, value=0}, {interpMode=RCIM_Cubic, tangentMode=RCTM_Break, time=1, value=1}]}}}
      CurveIndex=$Engine.System.Age
      Scale Curve=&Multiply_Float(A=0.8, B=&FNTimeOfDayScalar(Night Scale=1, Morning Scale=1, Day Scale=1.3, Evening Scale=1))
     )
     Passthrough Float=&FloatFromCurve(
      FloatCurve=DataInterface() {propertyValues {Curve {keys [{interpMode=RCIM_Cubic, tangentMode=RCTM_Break, time=1, value=0}, {interpMode=RCIM_Cubic, tangentMode=RCTM_Break, time=2, value=1}]}}}
      CurveIndex=$Engine.System.Age
      Scale Curve=0.8
     )
    )
    Scale RGB=&MaskVectorBySpawnGroup(
     Masked Vector=<1 1 1>
     Passthrough Vector=&MaskVectorBySpawnGroup(
      Masked Vector=<1 1 1>
      Passthrough Vector=&MakeVector(X=1, Y=1, Z=$User.Screen02_Light_Blue)
      Spawn Group Mask=1
     )
    )
    ScaleA=true
    ScaleRGB=true
   )
  ]
  rendererClasses [&NiagaraSpriteRendererProperties, &NiagaraLightRendererProperties]
  renderers [
   {rendererClass=&NiagaraSpriteRendererProperties}
   {rendererClass=&NiagaraSpriteRendererProperties, rendererIndex=1}
   {rendererClass=&NiagaraLightRendererProperties, rendererIndex=2}
  ]
 }
 Number_01 {rendererClasses [], renderers []}
 Number_02 {rendererClasses [], renderers []}
 Screen_02_Head {
  emitterUpdateScript [
   &EmitterState(Life Cycle Mode=::NewEnumerator1, Loop Duration=4, UseLoopDelay=false)
   &SpawnBurst_Instantaneous(Spawn Time=1, Use Spawn Probability=false, UseLoopCountLimit=false)
  ]
  particleSpawnScript [&InitializeParticle(
    Color Mode=::NewEnumerator0
    Lifetime=3
    Position Mode=::NewEnumerator0
    Sprite Size Mode=::NewEnumerator3
    Uniform Sprite Size=2
    Write Material Random=true
   )]
  particleUpdateScript [
   &ParticleState()
   &SystemLocation(MaskSpawnGroup=false, Offset=<0 -27.4 -9.5>) {module=SystemLocation001}
   &SystemLocation(
    MaskSpawnGroup=false
    Offset=&AddVectorToPosition(
     Position=$Particles.Position
     Vector=&MakeVector(
      X=0
      Y=&FloatFromCurve(FloatCurve=DataInterface() {propertyValues {Curve {keys [<at 0 0>,<at 1 13.7>]}}})
      Z=&FloatFromCurve(FloatCurve=DataInterface() {propertyValues {Curve {keys [
           {interpMode=RCIM_Cubic, leaveTangent=22.0712, tangentMode=RCTM_Break, time=0, value=0}
           {arriveTangent=5, interpMode=RCIM_Cubic, leaveTangent=5, time=0.1431, value=1.8721}
           {interpMode=RCIM_Cubic, tangentMode=RCTM_Break, time=1, value=5}
          ]}}})
     )
    )
   )
  ]
  rendererClasses [&NiagaraSpriteRendererProperties]
  renderers [{rendererClass=&NiagaraSpriteRendererProperties}]
 }
 Spaceship {
  emitterUpdateScript [
   &EmitterState(Life Cycle Mode=::NewEnumerator1, Loop Behavior=::NewEnumerator1, Loop Duration Mode=::NewEnumerator1, UseLoopDelay=false)
   &SpawnRate(SpawnRate=0.5, Use Spawn Probability=false)
   &SpawnRate(SpawnGroup=1, SpawnRate=2, Use Spawn Probability=false) {module=SpawnRate001}
   &SpawnRate(SpawnGroup=2, SpawnRate=2, Use Spawn Probability=false) {module=SpawnRate002}
   &SpawnRate(SpawnGroup=3, SpawnRate=2, Use Spawn Probability=false) {module=SpawnRate003}
  ]
  particleSpawnScript [
   /SetVariables(Particles.VisibilityTag=&SpawnGroup()) {isSetParameters=true}
   &InitializeParticle(
    Color=&MaskLinearColorBySpawnGroup(Masked Linear Color=<rgba 1 1 1 0.5>, Passthrough Linear Color=<rgba 0.079 0.509 1 1>)
    Lifetime=&MaskFloat(
     Mask Source=0
     Mask Parameter=?
     Mask=0
     Masked Float=2
     Passthrough Float=&RandomRangeFloat(Minimum=1.6, Maximum=1.8)
    )
    Position=&MaskVectorBySpawnGroup(
     Masked Vector=<0 19.7 1>
     Passthrough Vector=&MaskVectorBySpawnGroup(
      Masked Vector=<0 2 0>
      Passthrough Vector=&MaskVectorBySpawnGroup(Masked Vector=<0 -5 4>, Passthrough Vector=<0 6.5 6.5>, Spawn Group Mask=2)
      Spawn Group Mask=1
     )
    )
    Position Mode=::NewEnumerator1
    Sprite Size=&MaskVector2dBySpawnGroup(
     Masked Vector2D=<15.5 22.4>
     Passthrough Vector2D=&Vector2DFromFloat(Value=&RandomRangeFloat(Minimum=5, Maximum=8))
     Spawn Group Mask=0
    )
    Sprite Size Mode=::NewEnumerator4
    UsePositionOffset=false
    Write Material Random=true
   )
   &SpriteFacingAndAlignment(Sprite Alignment=<0 0 1>, WriteAlignment=true, WriteFacing=true)
  ]
  particleUpdateScript [
   &ParticleState()
   &ScaleSpriteSize(
    Scale Sprite Size Mode=::NewEnumerator0
    Uniform Scale Factor=&MaskFloat(
     Mask Source=0
     Mask Parameter=?
     Mask=0
     Masked Float=1
     Passthrough Float=&FloatFromCurve(FloatCurve=DataInterface() {propertyValues {Curve {keys [{interpMode=RCIM_Cubic, tangentMode=RCTM_Break, time=0, value=0}, {interpMode=RCIM_Cubic, tangentMode=RCTM_Break, time=1, value=1}]}}})
    )
   ) {module=ScaleSpriteSize001}
   &ScaleColor(
    Scale Alpha=&MaskFloat(
     Mask Source=0
     Mask Parameter=?
     Mask=0
     Masked Float=&FloatFromCurve(FloatCurve=DataInterface() {propertyValues {Curve {keys [
          {interpMode=RCIM_Cubic, tangentMode=RCTM_Break, time=0, value=0}
          {interpMode=RCIM_Cubic, tangentMode=RCTM_Break, time=0.5, value=1}
          {interpMode=RCIM_Cubic, tangentMode=RCTM_Break, time=1, value=0}
         ]}}})
     Passthrough Float=&FloatFromCurve(
      FloatCurve=DataInterface() {propertyValues {Curve {keys [{interpMode=RCIM_Cubic, tangentMode=RCTM_Break, time=0.4, value=1}, {interpMode=RCIM_Cubic, tangentMode=RCTM_Break, time=1, value=0}]}}}
      Scale Curve=&RandomRangeFloat(Minimum=0.8, Maximum=1)
     )
    )
    Scale RGB=&MaskVector(
     Mask Source=0
     Mask Parameter=?
     Mask=0
     Masked Vector=<1 1 1>
     Passthrough Vector=&VectorFromFloat(Value=&FloatFromCurve(FloatCurve=DataInterface() {propertyValues {Curve {keys [
           {interpMode=RCIM_Cubic, tangentMode=RCTM_Break, time=0.0899, value=5}
           {interpMode=RCIM_Cubic, tangentMode=RCTM_Break, time=0.2178, value=1.5}
           {interpMode=RCIM_Cubic, tangentMode=RCTM_Break, time=1, value=1}
          ]}}}))
    )
    ScaleA=true
    ScaleRGB=true
   )
  ]
  rendererClasses [&NiagaraSpriteRendererProperties]
  renderers [
   {rendererClass=&NiagaraSpriteRendererProperties}
   {rendererClass=&NiagaraSpriteRendererProperties, rendererIndex=1}
   {rendererClass=&NiagaraSpriteRendererProperties, rendererIndex=2}
   {rendererClass=&NiagaraSpriteRendererProperties, rendererIndex=3}
  ]
 }
 Text_01 {rendererClasses [], renderers []}
 Text_02 {rendererClasses [], renderers []}
 Text_03 {rendererClasses [], renderers []}
 Trail {
  emitterUpdateScript [
   &EmitterState(Life Cycle Mode=::NewEnumerator1, Loop Duration=12.5, UseLoopDelay=false)
   &SpawnParticlesFromOtherEmitter(
    Attribute Reader=DataInterface() {propertyValues {EmitterBinding {bindingMode=Other, emitterName=Head}}}
    Calculate Spawn Rate Per Particle=false
    Maximum Age Spawn Threshold=100000
    Minimum Age Spawn Threshold=0
    Spawn Rate=20
    Use Maximum Age=true
    Use Minimum Age=true
   )
  ]
  particleSpawnScript [
   &InitializeParticle(
    Color=<rgba 0.079 0.509 1 1>
    Lifetime=12.5
    Ribbon Facing Mode=::NewEnumerator1
    Ribbon Facing Vector:Vector3f=<1 0 1>
    Ribbon Width=0.3
    Ribbon Width Mode=::NewEnumerator1
    UsePositionOffset=false
    Write Material Random=true
   )
   &SampleParticlesFromOtherEmitter()
  ]
  particleUpdateScript [
   &ParticleState()
   &SolveForcesAndVelocity(Clamp Velocity=false, Limit Acceleration=false)
   &ScaleColor(
    Scale Alpha=&FloatFromCurve(FloatCurve=DataInterface() {propertyValues {Curve {keys [{interpMode=RCIM_Cubic, tangentMode=RCTM_Break, time=0, value=1}, <at 1 0>]}}})
    ScaleA=true
    ScaleRGB=true
   ) {module=ScaleColor001}
  ]
  rendererClasses [&NiagaraRibbonRendererProperties]
  renderers [{rendererClass=&NiagaraRibbonRendererProperties}]
 }
 Trail001 {
  emitterUpdateScript [
   &EmitterState(Life Cycle Mode=::NewEnumerator1, Loop Duration=4, UseLoopDelay=false)
   &SpawnParticlesFromOtherEmitter(
    Attribute Reader=DataInterface() {propertyValues {EmitterBinding {bindingMode=Other, emitterName=Screen_02_Head}}}
    Calculate Spawn Rate Per Particle=false
    Maximum Age Spawn Threshold=100000
    Minimum Age Spawn Threshold=0
    Use Maximum Age=true
    Use Minimum Age=true
   )
  ]
  particleSpawnScript [
   &InitializeParticle(
    Color=<rgba 0.0791 0.5089 1 1>
    Lifetime=4
    Ribbon Width=0.3
    Ribbon Width Mode=::NewEnumerator1
    Sprite Size Mode=::NewEnumerator3
    Uniform Sprite Size=1
    UsePositionOffset=false
    Write Material Random=true
   )
   &SampleParticlesFromOtherEmitter(Particle ID Sampling=::NewEnumerator1)
  ]
  particleUpdateScript [
   &ParticleState()
   &SolveForcesAndVelocity(Clamp Velocity=false, Limit Acceleration=false)
   &ScaleColor(
    Scale Alpha=&FloatFromCurve(FloatCurve=DataInterface() {propertyValues {Curve {keys [
         {interpMode=RCIM_Cubic, tangentMode=RCTM_Break, time=0, value=0}
         {interpMode=RCIM_Cubic, tangentMode=RCTM_Break, time=0.5, value=1}
         {interpMode=RCIM_Cubic, tangentMode=RCTM_Break, time=1, value=0}
        ]}}})
    ScaleA=true
    ScaleRGB=true
   )
   &ScaleRibbonWidth(Ribbon Width Scale=&FloatFromCurve(FloatCurve=DataInterface() {propertyValues {Curve {keys [
         {interpMode=RCIM_Cubic, tangentMode=RCTM_Break, time=0, value=0.3}
         {interpMode=RCIM_Cubic, tangentMode=RCTM_Break, time=0.3, value=1}
         {arriveTangent=0.0493, interpMode=RCIM_Cubic, leaveTangent=0.0493, tangentMode=RCTM_User, time=0.7, value=1}
         {interpMode=RCIM_Cubic, tangentMode=RCTM_Break, time=1, value=0.3}
        ]}}}))
  ]
  rendererClasses [&NiagaraRibbonRendererProperties]
  renderers [{rendererClass=&NiagaraRibbonRendererProperties}]
 }
 Trail003 {
  emitterUpdateScript [
   &EmitterState()
   &SpawnParticlesFromOtherEmitter(
    Attribute Reader=DataInterface() {propertyValues {EmitterBinding {bindingMode=Other, emitterName=Lead}}}
    Calculate Spawn Rate Per Particle=false
    Spawn Rate=60
    Use Maximum Age=false
    Use Minimum Age=false
   )
  ]
  particleSpawnScript [
   &InitializeParticle(
    Color=<rgba 0.2372 1.5266 3 1>
    Lifetime=1.2
    Ribbon Width=0.3
    Ribbon Width Mode=::NewEnumerator1
    UsePositionOffset=false
    Write Material Random=true
   )
   &SampleParticlesFromOtherEmitter(Emitter Sampling Mode:ENiagara_EmitterLocSamplingMode=::NewEnumerator0, Particle ID Sampling=::NewEnumerator1)
   &CameraOffset()
  ]
  particleUpdateScript [
   &ParticleState()
   &SolveForcesAndVelocity(Clamp Velocity=false, Limit Acceleration=false)
   &ScaleRibbonWidth(Ribbon Width Scale=&FloatFromCurve(FloatCurve=DataInterface() {propertyValues {Curve {keys [{interpMode=RCIM_Cubic, tangentMode=RCTM_Break, time=0, value=1}, {interpMode=RCIM_Cubic, tangentMode=RCTM_Break, time=1, value=0.3}]}}}))
   &ScaleColor(
    Scale Alpha=&FloatFromCurve(FloatCurve=DataInterface() {propertyValues {Curve {keys [{interpMode=RCIM_Cubic, tangentMode=RCTM_Break, time=0, value=1}, {interpMode=RCIM_Cubic, tangentMode=RCTM_Break, time=1, value=0}]}}})
    ScaleA=true
    ScaleRGB=true
   )
  ]
  rendererClasses [&NiagaraRibbonRendererProperties]
  renderers [{rendererClass=&NiagaraRibbonRendererProperties}]
 }
}
summary {
 systemName=NS_Emote_DonkeyCrib_Screen02_FX
 emitters [
  {emitterName=MainScreen, rendererClasses [&NiagaraSpriteRendererProperties, &NiagaraLightRendererProperties]}
  {emitterName=Screen_02_Head, rendererClasses [&NiagaraSpriteRendererProperties]}
  {emitterName=Trail001, rendererClasses [&NiagaraRibbonRendererProperties]}
  {emitterName=Lead, rendererClasses [&NiagaraSpriteRendererProperties]}
  {emitterName=Trail003, rendererClasses [&NiagaraRibbonRendererProperties]}
  {emitterName=Spaceship, rendererClasses [&NiagaraSpriteRendererProperties]}
  {emitterName=Head, rendererClasses [&NiagaraSpriteRendererProperties]}
  {emitterName=Trail, rendererClasses [&NiagaraRibbonRendererProperties]}
  {emitterName=Flare, rendererClasses []}
  {emitterName=Text_01, rendererClasses []}
  {emitterName=Bar, rendererClasses []}
  {emitterName=Number_01, rendererClasses []}
  {emitterName=Number_02, rendererClasses []}
  {emitterName=Text_02, rendererClasses []}
  {emitterName=Text_03, rendererClasses []}
 ]
 userVariables [User.Screen02_Light_Blue {type=NiagaraFloat, defaultValue=1, description=""}]
}
