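# Niagara system for the third screen effect in the DonkeyCrib emote, a glowing sci-fi display with formula and loading elements.
# MainScreen spawns four spawn-group layers, masked per group for color, position, rotation and opacity, with a scale-in, time-of-day brightness and a light renderer.
# Formula bursts two fading sprites on random loops, while Loading_Core grows a bar horizontally over its lifetime.
# User parameters drive element opacity, loading-core opacity and a blue-channel multiplier.
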NiagaraSystem /BRCosmetics/Animation/Game/MainPlayer/Emotes/DonkeyCrib/VFX/NS_Emote_DonkeyCrib_Screen03_FX
dependencies {
 usedDataInterfaces [&NiagaraDataInterfaceCurve, &NiagaraDataInterfaceVector2DCurve]
 usedDynamicInputs []
 usedModules [&SystemState]
}
emitters {
 Flare {rendererClasses [], renderers []}
 Formula {
  emitterUpdateScript [
   &EmitterState(
    Life Cycle Mode=::NewEnumerator1
    Loop Duration=&RandomRangeFloat(Minimum=5, Maximum=7)
    UseLoopDelay=false
   )
   &SpawnBurst_Instantaneous(Spawn Time=1, Use Spawn Probability=false, UseLoopCountLimit=false)
   &SpawnBurst_Instantaneous(Spawn Group=1, Spawn Time=1.5, Use Spawn Probability=false, UseLoopCountLimit=false) {module=SpawnBurst_Instantaneous001}
  ]
  particleSpawnScript [
   &InitializeParticle(
    Color=<rgba 0.323 1.585 3 1>
    Lifetime Max=5
    Lifetime Min=4.5
    Lifetime Mode=::NewEnumerator1
    Position=&MaskVectorBySpawnGroup(Masked Vector=<0 21 3>, Passthrough Vector=<0 -20.5 -4>)
    Position Mode=::NewEnumerator1
    Sprite Size=&MaskVector2dBySpawnGroup(Masked Vector2D=<15 16>, Passthrough Vector2D=<15 11.5>, Spawn Group Mask=0)
    Sprite Size Mode=::NewEnumerator4
    UsePositionOffset=false
    Write Material Random=true
   )
   /SetVariables() {isSetParameters=true}
   &SpriteFacingAndAlignment(Sprite Alignment=<0 0 1>, Sprite Facing=<-1 0 0>, WriteAlignment=true, WriteFacing=true)
  ]
  particleUpdateScript [
   &ParticleState()
   &ScaleColor(Scale Alpha=$User.Element_Opacity_Mult, ScaleA=true, ScaleRGB=true)
   &ScaleColor(
    Scale Alpha=&FloatFromCurve(FloatCurve=DataInterface() {propertyValues {Curve {keys [
         {interpMode=RCIM_Cubic, tangentMode=RCTM_Break, time=0, value=0}
         {interpMode=RCIM_Cubic, tangentMode=RCTM_Break, time=0.5, value=1}
         {interpMode=RCIM_Cubic, tangentMode=RCTM_Break, time=1, value=0}
        ]}}})
    ScaleA=true
    ScaleRGB=true
   ) {module=ScaleColor001}
  ]
  rendererClasses [&NiagaraSpriteRendererProperties]
  renderers [
   {rendererClass=&NiagaraSpriteRendererProperties}
   {rendererClass=&NiagaraSpriteRendererProperties, rendererIndex=1}
  ]
 }
 Formula_01 {bEnabled=false, rendererClasses [], renderers []}
 Formula_02 {bEnabled=false, rendererClasses [], renderers []}
 Loading_Core {
  emitterUpdateScript [
   &EmitterState(Life Cycle Mode=::NewEnumerator1, Loop Duration=14.8, UseLoopDelay=false)
   &SpawnBurst_Instantaneous(Spawn Time=1, Use Spawn Probability=false, UseLoopCountLimit=false)
  ]
  particleSpawnScript [
   &InitializeParticle(
    Color=<rgba 1.61 5.402 10 1>
    Lifetime=14.8
    Position=<0 -27 12>
    Position Mode=::NewEnumerator1
    Sprite Size=<1.25 0.56>
    Sprite Size Mode=::NewEnumerator4
    UsePositionOffset=false
    Write Material Random=true
   )
   &SpriteFacingAndAlignment(Sprite Alignment=<0 0 1>, WriteAlignment=true, WriteFacing=true)
  ]
  particleUpdateScript [
   &ParticleState()
   &ScaleSpriteSize(
    Non-Uniform Curve Sprite Scale=DataInterface() {propertyValues {
      XCurve {keys [{tangentMode=RCTM_User, time=0, value=0}, {tangentMode=RCTM_User, time=1, value=13.5}]}
      YCurve {keys [{tangentMode=RCTM_User, time=0, value=1}, {tangentMode=RCTM_User, time=1, value=1}]}
     }}
    Scale Sprite Size Mode=::NewEnumerator3
   )
   &ScaleColor(
    Scale Alpha=&FloatFromCurve(FloatCurve=DataInterface() {propertyValues {Curve {keys [<at 0.5 0.5>,<at 1 1>]}}})
    ScaleA=true
    ScaleRGB=true
   )
   &ScaleColor(Scale Alpha=$User.LoadingCore_Opacity_Mult, ScaleA=true, ScaleRGB=true) {module=ScaleColor001}
  ]
  rendererClasses [&NiagaraSpriteRendererProperties]
  renderers [{rendererClass=&NiagaraSpriteRendererProperties}]
 }
 MainScreen {
  emitterUpdateScript [
   &EmitterState(Life Cycle Mode=::NewEnumerator1, Loop Behavior=::NewEnumerator1, Loop Duration=14.8, UseLoopDelay=false)
   &SpawnBurst_Instantaneous(Use Spawn Probability=false, UseLoopCountLimit=false)
   &SpawnBurst_Instantaneous(Spawn Group=1, Spawn Time=1, Use Spawn Probability=false, UseLoopCountLimit=false) {module=SpawnBurst_Instantaneous001}
   &SpawnBurst_Instantaneous(Spawn Group=2, Spawn Time=1, Use Spawn Probability=false, UseLoopCountLimit=false) {module=SpawnBurst_Instantaneous002}
   &SpawnBurst_Instantaneous(Spawn Group=3, Use Spawn Probability=false, UseLoopCountLimit=false) {module=SpawnBurst_Instantaneous003}
  ]
  particleSpawnScript [
   &InitializeParticle(
    Color=&MaskLinearColor(
     Mask Source=0
     Mask Parameter=?
     Mask=0
     Masked Linear Color=<rgba 1 1 1 0.8>
     Passthrough Linear Color=&MaskLinearColorBySpawnGroup(
      Masked Linear Color=<rgba 0.197 0.631 1 0.7>
      Passthrough Linear Color=&MaskLinearColorBySpawnGroup(Masked Linear Color=<rgba 0.696 3.242 6 1>, Passthrough Linear Color=<rgba 0.696 3.242 6 1>, Spawn Group Mask=2)
      Spawn Group Mask=1
     )
    )
    Lifetime=5
    Position=&MaskVectorBySpawnGroup(
     Masked Vector=<0 0 0>
     Passthrough Vector=&MaskVector(
      Mask Source=0
      Mask Parameter=?
      Mask=0
      Masked Vector=<0 0 0>
      Passthrough Vector=&MaskVectorBySpawnGroup(Masked Vector=<0 -28 12>, Passthrough Vector=<0 0 0>, Spawn Group Mask=2)
     )
     Spawn Group Mask=1
    )
    Position Mode=::NewEnumerator1
    Sprite Size=<60 30>
    Sprite Size Mode=::NewEnumerator4
    UsePositionOffset=false
    Write Material Random=true
   )
   /SetVariables(Particles.SpriteFacing=<-1 0 0>) {isSetParameters=true}
   /SetVariables(Particles.SpriteAlignment=<0 0 1>) {isSetParameters=true}
   /SetVariables(Particles.VisibilityTag=&SpawnGroup()) {isSetParameters=true}
   /SetVariables(Particles.OrbSize:Vector2f=<25 25>) {isSetParameters=true}
   /SetVariables(Particles.LoadingBarSize:Vector2f=<18.75 2>) {isSetParameters=true}
  ]
  particleUpdateScript [
   &ParticleState(Kill Particles When Lifetime Has Elapsed=false)
   &SpriteRotationRate(Rotation Rate=&MaskFloat(
     Mask Source=0
     Mask Parameter=?
     Mask=0
     Masked Float=0
     Passthrough Float=&MaskFloatBySpawnGroup(Masked Float=15, Passthrough Float=0, Spawn Group Mask=1)
    ))
   &ScaleSpriteSize(
    Non-Uniform Curve Sprite Scale=DataInterface() {propertyValues {
      XCurve {keys [
        {interpMode=RCIM_Cubic, tangentMode=RCTM_Break, time=0, value=0}
        {arriveTangent=2.8331, interpMode=RCIM_Cubic, leaveTangent=25.7439, tangentMode=RCTM_Break, time=0.1498, value=0.1158}
        {arriveTangent=-0.0431, interpMode=RCIM_Cubic, tangentMode=RCTM_Break, time=0.1787, value=0.9863}
       ]}
      YCurve {keys [
        <at 0 0 RCIM_Cubic>
        {interpMode=RCIM_Cubic, leaveTangent=15.5999, tangentMode=RCTM_Break, time=0.0231, value=0.1813}
        {
         arriveTangentWeight=0.3394
         interpMode=RCIM_Cubic
         leaveTangent=31.0823
         leaveTangentWeight=0.078
         tangentMode=RCTM_Break
         time=0.1101
         value=0.8778
        }
        {arriveTangent=0.1208, interpMode=RCIM_Cubic, leaveTangent=-13.8768, tangentMode=RCTM_Break, time=0.1248, value=1.2343}
        {arriveTangent=-2.2265, interpMode=RCIM_Cubic, leaveTangent=-2.2265, tangentMode=RCTM_Break, time=0.1475, value=1}
        {interpMode=RCIM_Cubic, leaveTangent=1.8745, tangentMode=RCTM_Break, time=0.23, value=1}
       ]}
     }}
    Scale Sprite Size Mode=::NewEnumerator3
   )
   &ScaleColor(
    Scale Alpha=&MaskFloat(
     Mask Source=0
     Mask Parameter=?
     Mask=0
     Masked Float=&FloatFromCurve(
      FloatCurve=DataInterface() {propertyValues {Curve {keys [{interpMode=RCIM_Cubic, tangentMode=RCTM_Break, time=0, value=0}, {interpMode=RCIM_Cubic, tangentMode=RCTM_Break, time=1, value=1}]}}}
      CurveIndex=$System.Age
      Scale Curve=&Multiply_Float(A=1, B=&FNTimeOfDayScalar(Night Scale=1, Morning Scale=1, Day Scale=1.2, Evening Scale=1))
     )
     Passthrough Float=&MaskFloatBySpawnGroup(
      Masked Float=$User.Element_Opacity_Mult
      Passthrough Float=&MaskFloat(Mask Source=0, Mask Parameter=?, Mask=2, Masked Float=$User.Element_Opacity_Mult, Passthrough Float=1)
      Spawn Group Mask=1
     )
    )
    Scale RGB=&VectorFromFloat(Value=1)
    ScaleA=true
    ScaleRGB=true
   ) {module=ScaleColor001}
   &ScaleColor(
    Scale Alpha=&MaskFloatBySpawnGroup(
     Masked Float=1
     Passthrough Float=&FloatFromCurve(
      FloatCurve=DataInterface() {propertyValues {Curve {keys [{interpMode=RCIM_Cubic, tangentMode=RCTM_Break, time=1, value=0}, {interpMode=RCIM_Cubic, tangentMode=RCTM_Break, time=2, value=1}]}}}
      CurveIndex=$Engine.System.Age
     )
    )
    Scale RGB=&MaskVectorBySpawnGroup(
     Masked Vector=<1 1 1>
     Passthrough Vector=&MaskVectorBySpawnGroup(
      Masked Vector=<1 1 1>
      Passthrough Vector=&MaskVectorBySpawnGroup(
       Masked Vector=<1 1 1>
       Passthrough Vector=&MakeVector(X=1, Y=1, Z=$User.Screen03_Light_Blue)
       Spawn Group Mask=2
      )
      Spawn Group Mask=1
     )
    )
    ScaleA=true
    ScaleRGB=true
   ) {module=ScaleColor002}
   &ScaleColor(
    Scale Alpha=&MaskFloatBySpawnGroup(
     Masked Float=1
     Passthrough Float=&MaskFloatBySpawnGroup(
      Masked Float=&Multiply_Float(A=1, B=&FNTimeOfDayScalar(Night Scale=1, Morning Scale=1, Day Scale=1.3, Evening Scale=1))
      Passthrough Float=&MaskFloatBySpawnGroup(
       Masked Float=1
       Passthrough Float=&MaskFloatBySpawnGroup(Masked Float=1, Passthrough Float=1, Spawn Group Mask=3)
       Spawn Group Mask=2
      )
      Spawn Group Mask=1
     )
    )
    ScaleA=true
    ScaleRGB=true
   ) {module=ScaleColor003}
  ]
  rendererClasses [&NiagaraSpriteRendererProperties, &NiagaraLightRendererProperties]
  renderers [
   {rendererClass=&NiagaraSpriteRendererProperties}
   {rendererClass=&NiagaraSpriteRendererProperties, rendererIndex=1}
   {rendererClass=&NiagaraSpriteRendererProperties, rendererIndex=2}
   {rendererClass=&NiagaraLightRendererProperties, rendererIndex=3}
  ]
 }
}
summary {
 systemName=NS_Emote_DonkeyCrib_Screen03_FX
 emitters [
  {emitterName=MainScreen, rendererClasses [&NiagaraSpriteRendererProperties, &NiagaraLightRendererProperties]}
  {emitterName=Flare, rendererClasses []}
  {bEnabled=false, emitterName=Formula_01, rendererClasses []}
  {bEnabled=false, emitterName=Formula_02, rendererClasses []}
  {emitterName=Formula, rendererClasses [&NiagaraSpriteRendererProperties]}
  {emitterName=Loading_Core, rendererClasses [&NiagaraSpriteRendererProperties]}
 ]
 userVariables [
  User.Element_Opacity_Mult {type=NiagaraFloat, defaultValue=1, description=""}
  User.LoadingCore_Opacity_Mult {type=NiagaraFloat, defaultValue=1, description=""}
  User.Screen03_Light_Blue {type=NiagaraFloat, defaultValue=1, description=""}
 ]
}
