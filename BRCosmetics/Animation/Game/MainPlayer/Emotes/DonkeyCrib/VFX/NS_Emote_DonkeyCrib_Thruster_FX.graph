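# Niagara system for the thruster effect in the DonkeyCrib main player emote.
# It holds four emitters, named as two left and two right thrusters, with no renderers listed for any of them.
# System-level logic uses SystemState and a spawn-time variable assignment, with FNTimeOfDayColor as a dynamic input, suggesting time-of-day-aware color.
# It declares no user variables or data interfaces.
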NiagaraSystem /BRCosmetics/Animation/Game/MainPlayer/Emotes/DonkeyCrib/VFX/NS_Emote_DonkeyCrib_Thruster_FX
dependencies {
 usedDataInterfaces []
 usedDynamicInputs [&FNTimeOfDayColor]
 usedModules [
  &self:SystemSpawnScript.SystemScriptSource.SystemScriptGraph.NiagaraNodeAssignment_0.SetVariables_#1
  &SystemState
 ]
}
emitters {
 Thruster_Left_01 {rendererClasses [], renderers []}
 Thruster_Left_02 {rendererClasses [], renderers []}
 Thruster_Right_01 {rendererClasses [], renderers []}
 Thruster_Right_02 {rendererClasses [], renderers []}
}
summary {
 systemName=NS_Emote_DonkeyCrib_Thruster_FX
 emitters [
  {emitterName=Thruster_Left_01, rendererClasses []}
  {emitterName=Thruster_Left_02, rendererClasses []}
  {emitterName=Thruster_Right_01, rendererClasses []}
  {emitterName=Thruster_Right_02, rendererClasses []}
 ]
 userVariables []
}
