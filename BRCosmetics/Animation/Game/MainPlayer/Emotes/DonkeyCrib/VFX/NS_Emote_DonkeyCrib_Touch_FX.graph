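# Niagara touch effect for the main player's DonkeyCrib emote: a brief, bright cyan-blue sprite burst.
# A single emitter spawns an instantaneous burst of short-lived sprites with a Z alignment vector, material random and dynamic material parameters.
# Over each sprite's life, curves grow its size, flash its RGB early and fade its alpha out.
# User.Size sets the sprite size and User.Opacity scales the alpha fade.
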NiagaraSystem /BRCosmetics/Animation/Game/MainPlayer/Emotes/DonkeyCrib/VFX/NS_Emote_DonkeyCrib_Touch_FX
dependencies {usedDataInterfaces [&NiagaraDataInterfaceCurve], usedDynamicInputs [], usedModules [&SystemState]}
emitters {Touch001 {
  emitterUpdateScript [
   &EmitterState(Life Cycle Mode=::NewEnumerator1, Loop Behavior=::NewEnumerator1, Loop Duration Mode=::NewEnumerator1, UseLoopDelay=false)
   &SpawnBurst_Instantaneous(Use Spawn Probability=false, UseLoopCountLimit=false)
  ]
  particleSpawnScript [
   &InitializeParticle(
    Color=<rgba 0.6326 4.0709 8 1>
    Lifetime=0.45
    Position Mode=::NewEnumerator1
    Sprite Size Mode=::NewEnumerator3
    Uniform Sprite Size=$User.Size
    UsePositionOffset=false
    Write Material Random=true
   )
   /SetVariables(Particles.SpriteAlignment=<0 0 1>) {isSetParameters=true}
   /SetVariables() {isSetParameters=true}
   &DynamicMaterialParameters(Module.Index0Float0Write=true, Module.Index0Float1Write=true, Module.Index0Float2Write=false, Module.Index0Float3Write=false)
  ]
  particleUpdateScript [
   &ParticleState()
   &ScaleSpriteSize(Uniform Curve Sprite Scale=DataInterface() {propertyValues {Curve {keys [{interpMode=RCIM_Cubic, tangentMode=RCTM_Break, time=0, value=1}, {interpMode=RCIM_Cubic, tangentMode=RCTM_Break, time=1, value=1.5}]}}})
   &ScaleColor(
    Scale Alpha=&FloatFromCurve(
     FloatCurve=DataInterface() {propertyValues {Curve {keys [{interpMode=RCIM_Cubic, tangentMode=RCTM_Break, time=0, value=1}, {interpMode=RCIM_Cubic, tangentMode=RCTM_Break, time=1, value=0}]}}}
     Scale Curve=$User.Opacity
    )
    Scale RGB=&VectorFromFloat(Value=&FloatFromCurve(FloatCurve=DataInterface() {propertyValues {Curve {keys [
          {interpMode=RCIM_Cubic, tangentMode=RCTM_Break, time=0, value=1}
          {interpMode=RCIM_Cubic, tangentMode=RCTM_Break, time=0.2, value=3}
          {interpMode=RCIM_Cubic, tangentMode=RCTM_Break, time=1, value=1}
         ]}}}))
    ScaleA=true
    ScaleRGB=true
   )
  ]
  rendererClasses [&NiagaraSpriteRendererProperties]
  renderers [{rendererClass=&NiagaraSpriteRendererProperties}]
 }}
summary {
 systemName=NS_Emote_DonkeyCrib_Touch_FX
 emitters [{emitterName=Touch001, rendererClasses [&NiagaraSpriteRendererProperties]}]
 userVariables [
  User.Opacity {type=NiagaraFloat, defaultValue=1, description=""}
  User.Size {type=NiagaraFloat, defaultValue=5, description=""}
 ]
}
