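# NS_Emote_DonkeyCrib_Window_02 is a Niagara effect for the DonkeyCrib player emote, named as a window element of that emote.
# The system holds two emitters, an enabled Window and a disabled Light.
# Neither emitter lists renderers, and the only recorded module is SystemState, with no user variables, data interfaces or dynamic inputs.
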NiagaraSystem /BRCosmetics/Animation/Game/MainPlayer/Emotes/DonkeyCrib/VFX/NS_Emote_DonkeyCrib_Window_02
dependencies {usedDataInterfaces [], usedDynamicInputs [], usedModules [&SystemState]}
emitters {Light {bEnabled=false, rendererClasses [], renderers []}, Window {rendererClasses [], renderers []}}
summary {
 systemName=NS_Emote_DonkeyCrib_Window_02
 emitters [{emitterName=Window, rendererClasses []}, {bEnabled=false, emitterName=Light, rendererClasses []}]
 userVariables []
}
